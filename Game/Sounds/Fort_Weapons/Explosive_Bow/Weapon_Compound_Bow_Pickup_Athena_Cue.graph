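# Sound cue for the compound bow pickup sound in Athena, routed to the loot-collect sound class with weapon pickup attenuation.
# A quality-level root sends each pair of levels to its own mixer, delay and wave player chain.
# Only the second pair's delay has a set wait, under a second, and both mixers boost their wave while their second input is empty.
# A pickup concurrency set is assigned to limit overlapping plays.
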SoundCue /Game/Sounds/Fort_Weapons/Explosive_Bow/Weapon_Compound_Bow_Pickup_Athena_Cue
root=QualityLevel_1
nodes [
 {id=Delay_0, children [WavePlayer_2]}
 {id=Delay_1, children [WavePlayer_3], properties {delayMax=0.75, delayMin=0.75}}
 {id=Mixer_0, children [Delay_0,null], properties {inputVolume [2,0.4]}}
 {id=Mixer_1, children [Delay_1,null], properties {inputVolume [2,1]}}
 {id=QualityLevel_1, children [Mixer_0,Mixer_0,Mixer_1,Mixer_1]}
 {id=WavePlayer_2, properties {soundWaveAssetPtr=&/Weapon_Bow_Compound_Pickup}}
 {id=WavePlayer_3, properties {soundWaveAssetPtr=&Weapon_Bow_Compound_Pickup}}
]
properties {
 attenuationSettings=&Fort_Weapon_Pickup_Att
 maxDistance=3500
 soundClassObject=&Fort_Loot_Collect
 volumeMultiplier=0.4615
 concurrencyOverrides {maxCount=1, resolutionRule=PreventNew}
 concurrencySet [&PickupMaximum_Oldest]
}
