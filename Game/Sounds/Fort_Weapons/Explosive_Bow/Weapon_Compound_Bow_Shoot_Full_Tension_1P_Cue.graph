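# First-person sound cue for firing the explosive compound bow at full tension.
# A random node picks between two shot wave variations with equal weight.
# It uses first-person projectile attenuation, the first-person ranged-fire sound class and the shared first-person gunfire concurrency set.
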SoundCue /Game/Sounds/Fort_Weapons/Explosive_Bow/Weapon_Compound_Bow_Shoot_Full_Tension_1P_Cue
root=Random_0
nodes [
 {id=Random_0, children [WavePlayer_2,WavePlayer_3], properties {weights [1,1]}}
 {id=WavePlayer_2, properties {soundWaveAssetPtr=&/Weapon_Bow_Compound_Full_Tension_Shoot_01}}
 {id=WavePlayer_3, properties {soundWaveAssetPtr=&/Weapon_Bow_Compound_Full_Tension_Shoot_02}}
]
properties {
 attenuationSettings=&Fort_Weapon_Projectile_FirstPerson_Athena_Att
 maxDistance=2097152
 soundClassObject=&Fort_WeaponRanged_Fire_1P_PSM
 volumeMultiplier=1
 attenuationOverrides {falloffDistance=5000, attenuationShapeExtents {x=1000}}
 concurrencyOverrides {resolutionRule=PreventNew}
 concurrencySet [&Weapon_Gunfire_FirstPerson_Conc]
}
