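# First-person fire sound cue for a compound bow shot at low tension, filed under the Explosive Bow weapon sounds.
# A random node picks one of three equally weighted low-tension shot waves and feeds it through a modulator at the root.
# It uses first-person ranged-fire sound class, projectile attenuation and gunfire concurrency assets, with a reduced cue volume.
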SoundCue /Game/Sounds/Fort_Weapons/Explosive_Bow/Weapon_Compound_Bow_Shoot_Low_Tension_1P_Cue
root=Modulator_0
nodes [
 {id=Modulator_0, children [Random_1]}
 {id=Random_1, children [WavePlayer_0,WavePlayer_5,WavePlayer_6], properties {weights [1,1,1]}}
 {id=WavePlayer_0, properties {soundWaveAssetPtr=&/Weapon_Bow_Compound_Low_Tension_Shoot_01}}
 {id=WavePlayer_5, properties {soundWaveAssetPtr=&/Weapon_Bow_Compound_Low_Tension_Shoot_02}}
 {id=WavePlayer_6, properties {soundWaveAssetPtr=&/Weapon_Bow_Compound_Low_Tension_Shoot_03}}
]
properties {
 attenuationSettings=&Fort_Weapon_Projectile_FirstPerson_Athena_Att
 maxDistance=2097152
 soundClassObject=&Fort_WeaponRanged_Fire_1P_PSM
 volumeMultiplier=0.7
 attenuationOverrides {falloffDistance=5000, attenuationShapeExtents {x=1000}}
 concurrencyOverrides {resolutionRule=PreventNew}
 concurrencySet [&Weapon_Gunfire_FirstPerson_Conc]
}
