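# Sound cue for the first-person shot of a compound bow at normal tension, filed under the Explosive Bow weapon sounds.
# A Random node picks one of four equally weighted close-perspective shot waves, and a Modulator at default settings varies the result.
# Playback uses first-person projectile attenuation, the first-person ranged-fire sound class and first-person gunfire concurrency, with volume slightly reduced.
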SoundCue /Game/Sounds/Fort_Weapons/Explosive_Bow/Weapon_Compound_Bow_Shoot_Normal_Tension_1P_Cue
root=Modulator_0
nodes [
 {id=Modulator_0, children [Random_0]}
 {id=Random_0, children [WavePlayer_1,WavePlayer_2,WavePlayer_3,WavePlayer_4], properties {weights [1,1,1,1]}}
 {id=WavePlayer_1, properties {soundWaveAssetPtr=&/Weapon_Bow_Compound_Normal_Tension_Shoot_Close_01}}
 {id=WavePlayer_2, properties {soundWaveAssetPtr=&/Weapon_Bow_Compound_Normal_Tension_Shoot_Close_02}}
 {id=WavePlayer_3, properties {soundWaveAssetPtr=&/Weapon_Bow_Compound_Normal_Tension_Shoot_Close_03}}
 {id=WavePlayer_4, properties {soundWaveAssetPtr=&/Weapon_Bow_Compound_Normal_Tension_Shoot_Close_04}}
]
properties {
 attenuationSettings=&Fort_Weapon_Projectile_FirstPerson_Athena_Att
 maxDistance=2097152
 soundClassObject=&Fort_WeaponRanged_Fire_1P_PSM
 volumeMultiplier=0.8
 attenuationOverrides {falloffDistance=5000, attenuationShapeExtents {x=1000}}
 concurrencyOverrides {resolutionRule=PreventNew}
 concurrencySet [&Weapon_Gunfire_FirstPerson_Conc]
}
