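# Sound cue for firing the compound bow at normal tension, blending a close layer and a distant layer.
# The close layer picks randomly from four close-shot waves, while the distant layer plays a single medium shot wave, both with slight pitch modulation.
# Each layer has its own attenuation asset and sound class override before a mixer combines them, with the distant layer quieter.
# Local and global bow concurrency sets are assigned to the cue.
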SoundCue /Game/Sounds/Fort_Weapons/Explosive_Bow/Weapon_Compound_Bow_Shoot_Normal_Tension_Cue
root=Mixer_1
nodes [
 {
  id=Attenuation_3
  children [Modulator_0]
  properties {attenuationSettings=/Game/Sounds/Attenuation/Fort_Weapon_Projectile_Close_Bow_Athena_Att}
 }
 {
  id=Attenuation_4
  children [Modulator_1]
  properties {attenuationSettings=/Game/Sounds/Attenuation/Fort_Weapon_Projectile_Distant_Bow_Athena_Att}
 }
 {id=Mixer_1, children [SoundClass_3,SoundClass_2], properties {inputVolume [0.5,0.15]}}
 {id=Modulator_0, children [Random_0], properties {pitchMax=1.1}}
 {id=Modulator_1, children [WavePlayer_8], properties {pitchMax=1.1}}
 {id=Random_0, children [WavePlayer_3,WavePlayer_4,WavePlayer_5,WavePlayer_6], properties {weights [1,1,1,1]}}
 {
  id=SoundClass_2
  children [Attenuation_4]
  properties {soundClassOverride=&Fort_WeaponRanged_Fire_Distant}
 }
 {
  id=SoundClass_3
  children [Attenuation_3]
  properties {soundClassOverride=&Fort_WeaponRanged_Fire_3P_PSM}
 }
 {id=WavePlayer_3, properties {soundWaveAssetPtr=&/Weapon_Bow_Compound_Normal_Tension_Shoot_Close_01}}
 {id=WavePlayer_4, properties {soundWaveAssetPtr=&/Weapon_Bow_Compound_Normal_Tension_Shoot_Close_02}}
 {id=WavePlayer_5, properties {soundWaveAssetPtr=&/Weapon_Bow_Compound_Normal_Tension_Shoot_Close_03}}
 {id=WavePlayer_6, properties {soundWaveAssetPtr=&/Weapon_Bow_Compound_Normal_Tension_Shoot_Close_04}}
 {id=WavePlayer_8, properties {soundWaveAssetPtr=&/Weapon_Bow_Compound_Shoot_Med_01}}
]
properties {
 maxDistance=27500
 volumeMultiplier=1.2
 attenuationOverrides {falloffDistance=5000, attenuationShapeExtents {x=1000}}
 concurrencyOverrides {resolutionRule=PreventNew}
 concurrencySet [/Game/Sounds/Concurrency_Rules/Bow_Local_Conc, /Game/Sounds/Concurrency_Rules/Bow_Global_Conc]
}
