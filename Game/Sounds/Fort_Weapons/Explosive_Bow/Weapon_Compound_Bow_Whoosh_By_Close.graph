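# Sound cue for the close whoosh of a compound bow arrow passing by, in the explosive bow weapon sounds.
# A Random node picks one of four equally weighted arrow by-close waves and feeds it through a Modulator whose pitch range is pinned, so pitch does not vary.
# It routes to the ranged weapon projectile effects sound class, references a dedicated whoosh-by attenuation asset, overrides concurrency and has debug enabled.
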SoundCue /Game/Sounds/Fort_Weapons/Explosive_Bow/Weapon_Compound_Bow_Whoosh_By_Close
root=Modulator_0
nodes [
 {id=Modulator_0, children [Random_0], properties {pitchMax=1, pitchMin=1}}
 {id=Random_0, children [WavePlayer_1,WavePlayer_2,WavePlayer_3,WavePlayer_4], properties {weights [1,1,1,1]}}
 {id=WavePlayer_1, properties {soundWaveAssetPtr=&/Weapon_Bow_Compound_Arrow_By_Close_04}}
 {id=WavePlayer_2, properties {soundWaveAssetPtr=&/Weapon_Bow_Compound_Arrow_By_Close_01}}
 {id=WavePlayer_3, properties {soundWaveAssetPtr=&/Weapon_Bow_Compound_Arrow_By_Close_02}}
 {id=WavePlayer_4, properties {soundWaveAssetPtr=&/Weapon_Bow_Compound_Arrow_By_Close_03}}
]
properties {
 attenuationSettings=&/Weapon_Compound_Bow_Whoosh_By_Att
 bDebug=true
 bOverrideConcurrency=true
 maxDistance=2097152
 soundClassObject=&Fort_WeaponRanged_Projectile_Effects
 volumeMultiplier=1
 attenuationOverrides {
  bAttenuate=false
  bEnableOcclusion=true
  bEnableReverbSend=false
  falloffDistance=0
  occlusionLowPassFilterFrequency=2000
  occlusionTraceChannel=ECC_GameTraceChannel11
  occlusionVolumeAttenuation=0
  spatializationAlgorithm=SPATIALIZATION_HRTF
  attenuationShapeExtents {x=50000}
 }
}
